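; BPI_WorkflowItem is a Blueprint interface stub for items in the sound workflow system.
; It declares a single Initialize function that receives a BP Workflow Helper Data object reference.
; The function carries no logic here and returns false, leaving actual setup to implementers.
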(fn Initialize
  (params (WorkflowHelperData :type BP%20Workflow%20Helper%20Data%20Object%20Reference))
  (return false))
