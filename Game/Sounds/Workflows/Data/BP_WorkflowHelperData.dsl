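; BP_WorkflowHelperData is a Blueprint in the sound Workflows data folder that supplies a list of widgets to load, apparently for a workflow helper.
; Its one visible function, GetWidgetsToLoad, returns the stored GetWidgetsToLoad variable from the Default category, with no other logic.
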(fn GetWidgetsToLoad
  (return (Variables|Default|GetWidgetsToLoad)))
